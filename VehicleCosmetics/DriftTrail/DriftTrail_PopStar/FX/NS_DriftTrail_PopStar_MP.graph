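# Niagara system for the PopStar vehicle drift trail cosmetic, combining star meshes, colored star sprites and ribbon trails at the wheels.
# Particles follow skeletal mesh positions plus a user offset, and FN_DisableTireFX masks them per tire.
# Spawn groups give the head-star meshes steady, expanding-fade and flicker behaviors, while sprites fly backward and two widening ribbons stream behind.
# VC_PaintedVariantsOverride and bIsPainted drive per-paint colors and brightness.
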NiagaraSystem /VehicleCosmetics/DriftTrail/DriftTrail_PopStar/FX/NS_DriftTrail_PopStar_MP
dependencies {
 usedDataInterfaces [
  &NiagaraDataInterfaceCurve
  &NiagaraDataInterfaceArrayColor
  &NiagaraDataInterfaceArrayFloat3
  &NiagaraDataInterfaceArrayFloat
  &NiagaraDataInterfaceSkeletalMesh
 ]
 usedDynamicInputs []
 usedModules [
  &self:SystemSpawnScript.SystemScriptSource.SystemScriptGraph.NiagaraNodeAssignment_0.SetVariables_#1
  &SystemState
 ]
}
emitters {
 BCKP_Sprites_ColouredStars {
  emitterUpdateScript [
   &EmitterState()
   &SpawnRate(SpawnRate=25, Use Spawn Probability=false)
  ]
  particleSpawnScript [
   &VC_PaintedVariantsOverride(
    Black=&SelectLinearColorFromArray(
     Array Sampling Mode=::NewEnumerator0
     Direct Array Index=?
     Color Selection Array=DataInterface() {propertyValues {ColorData [<rgba 0.0191 0.0191 0.0191 1>, <rgba 0.45 0.45 0.45 1>]}}
     Random Seed=0
    )
    BurntSienna=&SelectLinearColorFromArray(
     Array Sampling Mode=::NewEnumerator0
     Direct Array Index=?
     Color Selection Array=DataInterface() {propertyValues {ColorData [<rgba 0.03 0.0093 0.0037 1>, <rgba 0.3739 0.088 0.0112 1>]}}
     Random Seed=0
    )
    Cobalt=&SelectLinearColorFromArray(
     Array Sampling Mode=::NewEnumerator0
     Direct Array Index=?
     Color Selection Array=DataInterface() {propertyValues {ColorData [<rgba 0 0.0185 1 1>, <rgba 0.32 0.3289 0.8 1>]}}
     Random Seed=0
    )
    Crimson=&SelectLinearColorFromArray(
     Array Sampling Mode=::NewEnumerator0
     Direct Array Index=?
     Color Selection Array=DataInterface() {propertyValues {ColorData [<rgba 0.4635 0 0.0126 1>, <rgba 1 0.0833 0 1>]}}
     Random Seed=0
    )
    DefaultColor=&SelectLinearColorFromArray(
     Array Sampling Mode=::NewEnumerator0
     Direct Array Index=?
     Color Selection Array=DataInterface() {propertyValues {ColorData [<rgba 0 1 0.1239 1>, <rgba 0.6939 0 1 1>, <rgba 1 0.5873 0.1156 1>, <rgba 0 0.4286 1 1>]}}
     Random Seed=0
    )
    ForestGreen=&SelectLinearColorFromArray(
     Array Sampling Mode=::NewEnumerator0
     Direct Array Index=?
     Color Selection Array=DataInterface() {propertyValues {ColorData [<rgba 0 0.5 0.0253 1>, <rgba 0 0.5 0.0253 1>]}}
     Random Seed=0
    )
    Grey=&SelectLinearColorFromArray(
     Array Sampling Mode=::NewEnumerator1
     Direct Array Index=0
     Color Selection Array=DataInterface() {propertyValues {ColorData [<rgba 0.2031 0.2031 0.2031 1>, <rgba 0.2031 0.2031 0.2031 1>]}}
     Random Seed=?
    )
    Lime=&SelectLinearColorFromArray(
     Array Sampling Mode=::NewEnumerator0
     Direct Array Index=?
     Color Selection Array=DataInterface() {propertyValues {ColorData [<rgba 0.325 0.65 0 1>, <rgba 0.325 0.65 0 1>]}}
     Random Seed=0
    )
    Orange=&SelectLinearColorFromArray(
     Array Sampling Mode=::NewEnumerator1
     Direct Array Index=0
     Color Selection Array=DataInterface() {propertyValues {ColorData [<rgba 0.6724 0.2346 0 1>, <rgba 0.6724 0.2346 0 1>]}}
     Random Seed=?
    )
    Pink=&SelectLinearColorFromArray(
     Array Sampling Mode=::NewEnumerator0
     Direct Array Index=?
     Color Selection Array=DataInterface() {propertyValues {ColorData [<rgba 0.5052 0.2779 0.4225 1>, <rgba 0.6 0.0719 0.408 1>]}}
     Random Seed=0
    )
    Purple=&SelectLinearColorFromArray(
     Array Sampling Mode=::NewEnumerator0
     Direct Array Index=?
     Color Selection Array=DataInterface() {propertyValues {ColorData [<rgba 0.2789 0 0.6795 1>, <rgba 0.3907 0.1019 0.6795 1>]}}
     Random Seed=0
    )
    Saffron=&SelectLinearColorFromArray(
     Array Sampling Mode=::NewEnumerator0
     Direct Array Index=?
     Color Selection Array=DataInterface() {propertyValues {ColorData [<rgba 0.5 0.3833 0 1>, <rgba 0.9 0.69 0 1>]}}
     Random Seed=0
    )
    SkyBlue=&SelectLinearColorFromArray(
     Array Sampling Mode=::NewEnumerator1
     Direct Array Index=0
     Color Selection Array=DataInterface() {propertyValues {ColorData [<rgba 0 0.357 0.5365 1>, <rgba 0 0.357 0.5365 1>]}}
     Random Seed=?
    )
    TitaniumWhite=&SelectLinearColorFromArray(
     Array Sampling Mode=::NewEnumerator0
     Direct Array Index=?
     Color Selection Array=DataInterface() {propertyValues {ColorData [<rgba 0.75 0.75 0.75 1>, <rgba 0.75 0.75 0.75 1>]}}
     Random Seed=0
    )
   )
   &InitializeParticle(
    Color=$Particles.Color
    Lifetime Max=0.5
    Lifetime Min=0.5
    Lifetime Mode=::NewEnumerator1
    Sprite Rotation Mode=::NewEnumerator2
    Sprite Size Mode=::NewEnumerator1
    Uniform Sprite Size Max=14
    Uniform Sprite Size Min=8
    UsePositionOffset=false
    Write Material Random=true
   )
   &PartitionParticles()
   &SkeletalMeshLocation(
    Bone / Socket Index=&Add_Integer(A=$Particles.PartitionParticles.Partition, B=4)
    Bone / Socket Sampling Mode=::NewEnumerator5
    Interpolate Mesh Sampling=false
    MaskSpawnGroup=false
    Mesh Sampling Type=::NewEnumerator6
    Sampled Position Offset=&AddVectorToPosition(Position=<-75 0 -20>, Vector=$User.offset)
    Skeletal Mesh=$System.SkeletalMesh
   )
   &AddVelocity(Velocity=&SelectVectorFromArray(
     Array Sampling Mode=::NewEnumerator0
     Direct Array Index=?
     Vector Selection Array=DataInterface() {propertyValues {FloatData [<-200 0 0>, <-500 1 1>]}}
     Random Seed=0
    ))
   &ShapeLocation(MaskSpawnGroup=false, Override Seed=false, Sphere Radius=15, UseOwnerScale=true)
  ]
  particleUpdateScript [
   &FN_DisableTireFX(
    Mask Index Attribute:NiagaraInt32=$Particles.PartitionParticles.Partition
    Mask Index FL=2
    Mask Index FR=3
    Mask Index RL=0
    Mask Index RR=1
   )
   &ParticleState()
   &ScaleColor(
    Scale Alpha=&FloatFromCurve(FloatCurve=DataInterface() {propertyValues {Curve {keys [<at 0 0>,<at 0.2 1>,<at 1 0>]}}})
    Scale RGB=&MakeCustomVectorFromBool(
     Bool=$User.bIsPainted
     True Vector=&SelectVectorFromArray(
      Array Sampling Mode=::NewEnumerator0
      Direct Array Index=?
      Vector Selection Array=DataInterface() {propertyValues {FloatData [<0 0 0>, <4 4 4>]}}
      Random Seed=0
     )
     False Vector=&VectorFromFloat(Value=4)
    )
    ScaleA=true
    ScaleRGB=true
   ) {enabled=false}
   &ScaleColor(
    Scale Alpha=&FloatFromCurve(FloatCurve=DataInterface() {propertyValues {Curve {keys [<at 0 0>,<at 0.2 1>,<at 1 0>]}}})
    Scale RGB=&MakeCustomVectorFromBool(
     Bool=$User.bIsPainted
     True Vector=&Multiply_VectorByFloat(
      Vector=&SelectVectorFromArray(
       Array Sampling Mode=::NewEnumerator0
       Direct Array Index=?
       Vector Selection Array=DataInterface() {propertyValues {FloatData [<0.5 0.5 0.5>, <7 7 7>]}}
       Random Seed=0
      )
      Float=$User.StarsSpritesScalar
     )
     False Vector=&VectorFromFloat(Value=4)
    )
    ScaleA=true
    ScaleRGB=true
   ) {module=ScaleColor002}
   &SolveForcesAndVelocity(Clamp Velocity=false, Limit Acceleration=false)
  ]
  rendererClasses [&NiagaraSpriteRendererProperties]
  renderers [{rendererClass=&NiagaraSpriteRendererProperties}]
 }
 BCKP_StaticStar {
  emitterSpawnScript [
   /SetVariables(Emitter.No Painted:NiagaraBool=&Boolean_NotOperation(Bool=$User.bIsPainted)) {isSetParameters=true}
   /SetVariables(Emitter.Painted:NiagaraBool=&Boolean_NotOperation(Bool=$StackContext.No Painted)) {isSetParameters=true}
  ]
  emitterUpdateScript [
   &EmitterState(Inactive Response=::NewEnumerator1, Life Cycle Mode=::NewEnumerator1, Loop Duration=1, UseLoopDelay=false)
   &SpawnRate(SpawnGroup=1, SpawnRate=8, Use Spawn Probability=false) {enabled=false}
   &SpawnBurst_Instantaneous(Spawn Count=3, Use Spawn Probability=false, UseLoopCountLimit=false)
  ]
  particleSpawnScript [
   /SetVariables(Particles.VisTag:NiagaraInt32=&SpawnGroup()) {isSetParameters=true, module=SetVariables_#2}
   &InitializeParticle(
    Color Mode=::NewEnumerator0
    Lifetime=/VehicleCosmetics/FX/Niagara/Modules/MergeEmitters/VC_FloatBySpawnGroup(
     SpawnGroup or ExecIndex:NiagaraInt32=0
     CustomInt:NiagaraInt32=0
     MaximumSpawnGroups:NiagaraInt32=3
     Float SG 0:NiagaraFloat=$Emitter.CurrentLoopDuration
     Float SG 1:NiagaraFloat=0.5
     Float SG 2:NiagaraFloat=$Emitter.CurrentLoopDuration
     Float SG 3:NiagaraFloat=0
     Float SG 4:NiagaraFloat=0
     Float SG 5:NiagaraFloat=0
    )
    Mesh Scale Mode=::NewEnumerator3
    Mesh Uniform Scale=&VC_FloatBySpawnGroup(
     SpawnGroup or ExecIndex:NiagaraInt32=0
     CustomInt:NiagaraInt32=0
     MaximumSpawnGroups:NiagaraInt32=3
     Float SG 0:NiagaraFloat=2.5
     Float SG 1:NiagaraFloat=2.5
     Float SG 2:NiagaraFloat=3
     Float SG 3:NiagaraFloat=0
     Float SG 4:NiagaraFloat=0
     Float SG 5:NiagaraFloat=0
    )
    Sprite Rotation Mode=::NewEnumerator2
    UsePositionOffset=false
    Write Material Random=true
   )
   &PartitionParticles()
   &SkeletalMeshLocation(
    Bone / Socket Index=&Add_Integer(A=$Particles.PartitionParticles.Partition, B=4)
    Bone / Socket Sampling Mode=::NewEnumerator5
    Interpolate Mesh Sampling=false
    MaskSpawnGroup=false
    Mesh Sampling Type=::NewEnumerator6
    Sampled Position Offset=&AddVectorToPosition(
     Position=&MakeVector(
      X=&VC_FloatBySpawnGroup(
       SpawnGroup or ExecIndex:NiagaraInt32=0
       CustomInt:NiagaraInt32=0
       MaximumSpawnGroups:NiagaraInt32=3
       Float SG 0:NiagaraFloat=-61
       Float SG 1:NiagaraFloat=-63
       Float SG 2:NiagaraFloat=-63
       Float SG 3:NiagaraFloat=0
       Float SG 4:NiagaraFloat=0
       Float SG 5:NiagaraFloat=0
      )
      Y=&VC_FloatBySpawnGroup(
       SpawnGroup or ExecIndex:NiagaraInt32=0
       CustomInt:NiagaraInt32=0
       MaximumSpawnGroups:NiagaraInt32=3
       Float SG 0:NiagaraFloat=0
       Float SG 1:NiagaraFloat=0
       Float SG 2:NiagaraFloat=0
       Float SG 3:NiagaraFloat=0
       Float SG 4:NiagaraFloat=0
       Float SG 5:NiagaraFloat=0
      )
      Z=&VC_FloatBySpawnGroup(
       SpawnGroup or ExecIndex:NiagaraInt32=0
       CustomInt:NiagaraInt32=0
       MaximumSpawnGroups:NiagaraInt32=3
       Float SG 0:NiagaraFloat=-20
       Float SG 1:NiagaraFloat=-20
       Float SG 2:NiagaraFloat=-20
       Float SG 3:NiagaraFloat=0
       Float SG 4:NiagaraFloat=0
       Float SG 5:NiagaraFloat=0
      )
     )
     Vector=$User.offset
    )
    Skeletal Mesh=$System.SkeletalMesh
   )
   &VC_PaintedVariantsOverride(
    Black=<rgba 0.02 0.02 0.02 1>
    BurntSienna=<rgba 0.2 0.0618 0.0247 1>
    Cobalt=<rgba 0 0.0148 0.8 1>
    Crimson=<rgba 0.25 0 0.0068 1>
    DefaultColor=<rgba 0.6261 0.6261 0.6261 1>
    ForestGreen=<rgba 0 0.4 0.0202 1>
    Grey=<rgba 0.21 0.21 0.21 1>
    Lime=<rgba 0.3374 0.75 0.1111 1>
    Orange=<rgba 0.45 0.108 0.0225 1>
    Pink=<rgba 0.6 0.24 0.5213 1>
    Purple=<rgba 0.2462 0 0.6 1>
    R1 BurntSienna:LinearColor=<rgba 0.15 0.0463 0.0185 1>
    R1 DefaultColor=<rgba 0.45 0.45 0.45 1>
    R1 Lime:LinearColor=<rgba 0.2943 0.5885 0 0.8>
    R1 Orange:LinearColor=<rgba 0.6724 0.2346 0 0.5>
    R1 Pink:LinearColor=<rgba 0.4 0.16 0.3127 1>
    R1 Purple:LinearColor=<rgba 0.4399 0.05 1 0.8>
    R1 Saffron:LinearColor=<rgba 0.9 0.7005 0.045 1>
    R1 SkyBlue:LinearColor=<rgba 0 0.2019 0.6 1>
    R2 Black:LinearColor=<rgba 0.08 0.08 0.08 1>
    R2 BurntSienna:LinearColor=<rgba 0.5 0.1255 0.025 1>
    R2 Cobalt:LinearColor=<rgba 0 0.0185 1 1>
    R2 DefaultColor:LinearColor=<rgba 1 0.4 1 1>
    R2 Grey:LinearColor=<rgba 0.4 0.4 0.4 1>
    R2 Orange:LinearColor=<rgba 0.6724 0.1681 0 1>
    R2 Pink:LinearColor=<rgba 0.8 0.36 0.64 1>
    R2 Purple:LinearColor=<rgba 0.5632 0.12 1.2 1>
    R2 Saffron:LinearColor=<rgba 1 0.79 0.1 1>
    R2 SkyBlue:LinearColor=<rgba 0.05 0.6821 1 1>
    R3 Black:LinearColor=<rgba 0.12 0.12 0.12 1>
    R3 BurntSienna:LinearColor=<rgba 1 0.25 0 1>
    R3 Cobalt:LinearColor=<rgba 0.1 0.1351 2 1>
    R3 Crimson:LinearColor=<rgba 0.6 0.03 0 1>
    R3 DefaultColor:LinearColor=<rgba 3 1.2 3 1>
    R3 ForestGreen:LinearColor=<rgba 0 1.5 0.0758 1>
    R3 Grey:LinearColor=<rgba 1 1 1 1>
    R3 Lime:LinearColor=<rgba 1 2 0 1>
    R3 Orange:LinearColor=<rgba 2 0.3333 0 1>
    R3 Pink:LinearColor=<rgba 2 0.9 1.6 1>
    R3 Purple:LinearColor=<rgba 0.8208 0 2 1>
    R3 Saffron:LinearColor=<rgba 1 0.8017 0.15 1>
    R3 SkyBlue:LinearColor=<rgba 0.1 1.3642 2 1>
    Saffron=<rgba 0.3 0.3 0 1>
    SkyBlue=<rgba 0 0.4167 1 1>
    TitaniumWhite=<rgba 0.55 0.55 0.55 1>
    Use R1 Painted Overrides=true
    Use R2 Painted Overrides=true
    Use R3 Painted Overrides=true
   )
  ]
  particleUpdateScript [
   &FN_DisableTireFX(
    Mask Index Attribute:NiagaraInt32=$Particles.PartitionParticles.Partition
    Mask Index FL=2
    Mask Index FR=3
    Mask Index RL=0
    Mask Index RR=1
   )
   &ParticleState()
   &ScaleMeshSize(Scale Factor=/VehicleCosmetics/FX/Niagara/Modules/MergeEmitters/VC_VectorBySpawnGroup(
     SpawnGroup or ExecIndex:NiagaraInt32=0
     CustomInt:NiagaraInt32=0
     MaximumSpawnGroups:NiagaraInt32=3
     Vector Group 0:Vector3f=<1 1 1>
     Vector Group 1:Vector3f=&VectorFromFloat(Value=&FloatFromCurve(
       FloatCurve=DataInterface() {propertyValues {Curve {keys [{arriveTangent=-0.0233, interpMode=RCIM_Cubic, leaveTangent=-0.0233, tangentMode=RCTM_User, time=0, value=0.5}, <at 1 1 RCIM_Cubic>]}}}
       Scale Curve=1.2
      ))
     Vector Group 2:Vector3f=&VectorFromFloat(Value=&FloatFromCurve(
       FloatCurve=DataInterface() {propertyValues {Curve {keys [<at 0 1.2>,<at 0.025 0.85>,<at 0.05 1.2>,<at 0.075 0.85>,<at 0.1 1.2>,<at 0.125 0.85>,<at 0.15 1.2>,<at 0.175 0.85>,<at 0.2 1.2>,<at 0.225 0.85 RCIM_Cubic>,<at 0.25 1.2 RCIM_Cubic>,<at 0.275 0.85 RCIM_Cubic>,<at 0.3 1.2 RCIM_Cubic>,<at 0.325 0.85 RCIM_Cubic>,<at 0.35 1.2 RCIM_Cubic>,<at 0.375 0.85 RCIM_Cubic>,<at 0.4 1.2 RCIM_Cubic>,<at 0.425 0.85 RCIM_Cubic>,<at 0.45 1.2 RCIM_Cubic>,<at 0.475 0.85 RCIM_Cubic>,<at 0.5 1.2 RCIM_Cubic>,<at 0.525 0.85 RCIM_Cubic>,<at 0.55 1.2 RCIM_Cubic>,<at 0.575 0.85 RCIM_Cubic>,<at 0.6 1.2 RCIM_Cubic>,<at 0.625 0.85 RCIM_Cubic>,<at 0.65 1.2 RCIM_Cubic>,<at 0.675 0.85 RCIM_Cubic>,<at 0.7 1.2 RCIM_Cubic>,<at 0.725 0.85 RCIM_Cubic>,<at 0.75 1.2 RCIM_Cubic>,<at 0.775 0.85 RCIM_Cubic>,<at 0.8 1.2 RCIM_Cubic>,<at 0.825 0.85 RCIM_Cubic>,<at 0.85 1.2 RCIM_Cubic>,<at 0.875 0.85 RCIM_Cubic>,<at 0.9 1.2 RCIM_Cubic>,<at 0.925 0.85 RCIM_Cubic>,<at 0.95 1.2 RCIM_Cubic>,<at 0.975 0.85 RCIM_Cubic>,<at 1 1.2 RCIM_Cubic>]}}}
       Scale Curve=0.8
      ))
     Vector Group 3:Vector3f=<0 0 0>
     Vector Group 4:Vector3f=<0 0 0>
     Vector Group 5:Vector3f=<0 0 0>
    ))
   &DynamicMaterialParameters(
    Index 0 Param 1=1
    Index 0 Param 2=&MakeCustomFloatFromBool(Bool=$User.bIsPainted, True Float=1, False Float=0)
    Module.Index0Param0Write=true
    Module.Index0Param1Write=true
    Module.Index0Param2Write=true
    Module.Index0Param3Write=true
    Module.Index3Param0Write=true
    Module.Index3Param1Write=true
    Module.Index3Param2Write=true
    Module.Index3Param3Write=true
    Write Parameter Index 3=true
   )
   &ScaleColor(
    Scale Alpha=&VC_FloatBySpawnGroup(
     SpawnGroup or ExecIndex:NiagaraInt32=0
     CustomInt:NiagaraInt32=0
     MaximumSpawnGroups:NiagaraInt32=3
     Float SG 0:NiagaraFloat=1
     Float SG 1:NiagaraFloat=&FloatFromCurve(FloatCurve=DataInterface() {propertyValues {Curve {keys [<at 0 1>,<at 1 0>]}}})
     Float SG 2:NiagaraFloat=1
     Float SG 3:NiagaraFloat=0
     Float SG 4:NiagaraFloat=0
     Float SG 5:NiagaraFloat=0
    )
    Scale RGB=&VC_VectorBySpawnGroup(
     SpawnGroup or ExecIndex:NiagaraInt32=0
     CustomInt:NiagaraInt32=0
     MaximumSpawnGroups:NiagaraInt32=3
     Vector Group 0:Vector3f=<0.7 0.7 0.7>
     Vector Group 1:Vector3f=&VectorFromFloat(Value=&MakeCustomFloatFromBool(Bool=$User.bIsPainted, True Float=$User.StarOutlineScalar, False Float=5))
     Vector Group 2:Vector3f=&VectorFromFloat(Value=&MakeCustomFloatFromBool(Bool=$User.bIsPainted, True Float=&Multiply_Float(A=$User.StarOutlineScalar, B=1), False Float=5))
     Vector Group 3:Vector3f=<0 0 0>
     Vector Group 4:Vector3f=<0 0 0>
     Vector Group 5:Vector3f=<0 0 0>
    )
    ScaleA=true
    ScaleRGB=true
   )
   &SolveForcesAndVelocity(Clamp Velocity=false, Limit Acceleration=false)
  ]
  rendererClasses [&NiagaraMeshRendererProperties, &NiagaraSpriteRendererProperties]
  renderers [
   {rendererClass=&NiagaraMeshRendererProperties, rendererIndex=0}
   {rendererClass=&NiagaraMeshRendererProperties, rendererIndex=1}
   {rendererClass=&NiagaraMeshRendererProperties, rendererIndex=2}
   {rendererClass=&NiagaraMeshRendererProperties, rendererIndex=3}
   {rendererClass=&NiagaraSpriteRendererProperties, rendererIndex=4}
  ]
 }
 BCKP_mesh_headstarr001 {
  emitterUpdateScript [
   &EmitterState(Inactive Response=::NewEnumerator1, Life Cycle Mode=::NewEnumerator1, Loop Duration=1, UseLoopDelay=false)
   &SpawnRate(SpawnGroup=1, SpawnRate=8, Use Spawn Probability=false)
   &SpawnBurst_Instantaneous(Spawn Count=3, Use Spawn Probability=false, UseLoopCountLimit=false)
   &SpawnBurst_Instantaneous(Spawn Count=2, Spawn Group=2, Use Spawn Probability=false, UseLoopCountLimit=false) {module=SpawnBurst_Instantaneous001}
  ]
  particleSpawnScript [
   /SetVariables(Particles.VisTag:NiagaraInt32=&SpawnGroup()) {isSetParameters=true, module=SetVariables_#2}
   &InitializeParticle(
    Color Mode=::NewEnumerator0
    Lifetime=&VC_FloatBySpawnGroup(
     SpawnGroup or ExecIndex:NiagaraInt32=0
     CustomInt:NiagaraInt32=0
     MaximumSpawnGroups:NiagaraInt32=3
     Float SG 0:NiagaraFloat=$Emitter.CurrentLoopDuration
     Float SG 1:NiagaraFloat=0.5
     Float SG 2:NiagaraFloat=$Emitter.CurrentLoopDuration
     Float SG 3:NiagaraFloat=0
     Float SG 4:NiagaraFloat=0
     Float SG 5:NiagaraFloat=0
    )
    Mesh Scale Mode=::NewEnumerator3
    Mesh Uniform Scale=&VC_FloatBySpawnGroup(
     SpawnGroup or ExecIndex:NiagaraInt32=0
     CustomInt:NiagaraInt32=0
     MaximumSpawnGroups:NiagaraInt32=3
     Float SG 0:NiagaraFloat=2.5
     Float SG 1:NiagaraFloat=2.5
     Float SG 2:NiagaraFloat=3
     Float SG 3:NiagaraFloat=0
     Float SG 4:NiagaraFloat=0
     Float SG 5:NiagaraFloat=0
    )
    Sprite Rotation Mode=::NewEnumerator2
    UsePositionOffset=false
    Write Material Random=true
   )
   &PartitionParticles()
   &SkeletalMeshLocation(
    Bone / Socket Index=&Add_Integer(A=$Particles.PartitionParticles.Partition, B=4)
    Bone / Socket Sampling Mode=::NewEnumerator5
    Interpolate Mesh Sampling=false
    MaskSpawnGroup=false
    Mesh Sampling Type=::NewEnumerator6
    Sampled Position Offset=&AddVectorToPosition(
     Position=&MakeVector(
      X=&VC_FloatBySpawnGroup(
       SpawnGroup or ExecIndex:NiagaraInt32=0
       CustomInt:NiagaraInt32=0
       MaximumSpawnGroups:NiagaraInt32=3
       Float SG 0:NiagaraFloat=-61
       Float SG 1:NiagaraFloat=-63
       Float SG 2:NiagaraFloat=-63
       Float SG 3:NiagaraFloat=0
       Float SG 4:NiagaraFloat=0
       Float SG 5:NiagaraFloat=0
      )
      Y=&VC_FloatBySpawnGroup(
       SpawnGroup or ExecIndex:NiagaraInt32=0
       CustomInt:NiagaraInt32=0
       MaximumSpawnGroups:NiagaraInt32=3
       Float SG 0:NiagaraFloat=0
       Float SG 1:NiagaraFloat=0
       Float SG 2:NiagaraFloat=0
       Float SG 3:NiagaraFloat=0
       Float SG 4:NiagaraFloat=0
       Float SG 5:NiagaraFloat=0
      )
      Z=&VC_FloatBySpawnGroup(
       SpawnGroup or ExecIndex:NiagaraInt32=0
       CustomInt:NiagaraInt32=0
       MaximumSpawnGroups:NiagaraInt32=3
       Float SG 0:NiagaraFloat=-20
       Float SG 1:NiagaraFloat=-20
       Float SG 2:NiagaraFloat=-20
       Float SG 3:NiagaraFloat=0
       Float SG 4:NiagaraFloat=0
       Float SG 5:NiagaraFloat=0
      )
     )
     Vector=$User.offset
    )
    Skeletal Mesh=$System.SkeletalMesh
   )
   &VC_PaintedVariantsOverride(
    Black=<rgba 0.009 0.009 0.009 1>
    BurntSienna=<rgba 0.0642 0.0198 0.0079 1>
    Cobalt=<rgba 0 0.0129 0.7 1>
    Crimson=<rgba 0.1 0 0.0027 1>
    ForestGreen=<rgba 0 0.5 0.0253 1>
    Grey=<rgba 0 0.045 0.05 1>
    Lime=<rgba 0.1917 0.4261 0.0631 1>
    Orange=<rgba 0.3 0.1046 0 1>
    Pink=<rgba 0.4 0.0956 0.3335 1>
    Purple=<rgba 0.1585 0.06 0.3 1>
    R1 BurntSienna:LinearColor=<rgba 0.15 0.0463 0.0185 1>
    R1 DefaultColor=<rgba 0.5 0 0.5 1.5>
    R1 Lime:LinearColor=<rgba 0.2943 0.5885 0 0.8>
    R1 Orange:LinearColor=<rgba 0.6724 0.2346 0 0.5>
    R1 Pink:LinearColor=<rgba 0.4 0.16 0.3127 1>
    R1 Purple:LinearColor=<rgba 0.4399 0.05 1 0.8>
    R1 Saffron:LinearColor=<rgba 0.9 0.7005 0.045 1>
    R1 SkyBlue:LinearColor=<rgba 0 0.2019 0.6 1>
    R2 Black:LinearColor=<rgba 0.08 0.08 0.08 1>
    R2 BurntSienna:LinearColor=<rgba 0.5 0.1255 0.025 1>
    R2 Cobalt:LinearColor=<rgba 0.0167 0 1 1>
    R2 DefaultColor:LinearColor=<rgba 1 0.4 1 1>
    R2 Grey:LinearColor=<rgba 0.4 0.4 0.4 1>
    R2 Orange:LinearColor=<rgba 0.6724 0.1681 0 1>
    R2 Pink:LinearColor=<rgba 0.8 0.36 0.64 1>
    R2 Purple:LinearColor=<rgba 0.5632 0.12 1.2 1>
    R2 Saffron:LinearColor=<rgba 1 0.79 0.1 1>
    R2 SkyBlue:LinearColor=<rgba 0.05 0.6821 1 1>
    R3 Black:LinearColor=<rgba 0.12 0.12 0.12 1>
    R3 BurntSienna:LinearColor=<rgba 0.5 0.125 0 1>
    R3 Cobalt:LinearColor=<rgba 0.0167 0 1 1>
    R3 Crimson:LinearColor=<rgba 0.6 0.03 0 1>
    R3 DefaultColor:LinearColor=<rgba 3 1.2 3 1>
    R3 ForestGreen:LinearColor=<rgba 0 1.5 0.0758 1>
    R3 Grey:LinearColor=<rgba 1 1 1 1>
    R3 Lime:LinearColor=<rgba 1 2 0 1>
    R3 Orange:LinearColor=<rgba 2 0.3333 0 1>
    R3 Pink:LinearColor=<rgba 2 0.9 1.6 1>
    R3 Purple:LinearColor=<rgba 0.8208 0 2 1>
    R3 Saffron:LinearColor=<rgba 1 0.8017 0.15 1>
    R3 SkyBlue:LinearColor=<rgba 0.1 1.3642 2 1>
    Use R1 Painted Overrides=true
    Use R2 Painted Overrides=true
    Use R3 Painted Overrides=true
   )
  ]
  particleUpdateScript [
   &FN_DisableTireFX(
    Mask Index Attribute:NiagaraInt32=$Particles.PartitionParticles.Partition
    Mask Index FL=2
    Mask Index FR=3
    Mask Index RL=0
    Mask Index RR=1
   )
   &ParticleState()
   &ScaleMeshSize(Scale Factor=&VC_VectorBySpawnGroup(
     SpawnGroup or ExecIndex:NiagaraInt32=0
     CustomInt:NiagaraInt32=0
     MaximumSpawnGroups:NiagaraInt32=3
     Vector Group 0:Vector3f=<1 1 1>
     Vector Group 1:Vector3f=&VectorFromFloat(Value=&FloatFromCurve(
       FloatCurve=DataInterface() {propertyValues {Curve {keys [{arriveTangent=-0.0233, interpMode=RCIM_Cubic, leaveTangent=-0.0233, tangentMode=RCTM_User, time=0, value=0.5}, <at 1 1 RCIM_Cubic>]}}}
       Scale Curve=1.2
      ))
     Vector Group 2:Vector3f=&VectorFromFloat(Value=&FloatFromCurve(
       FloatCurve=DataInterface() {propertyValues {Curve {keys [<at 0 1.2>,<at 0.025 0>,<at 0.05 1.2>,<at 0.075 0>,<at 0.1 1.2>,<at 0.125 0>,<at 0.15 1.2>,<at 0.175 0>,<at 0.2 1.2>,<at 0.225 0 RCIM_Cubic>,<at 0.25 1.2 RCIM_Cubic>,<at 0.275 0 RCIM_Cubic>,<at 0.3 1.2 RCIM_Cubic>,<at 0.325 0 RCIM_Cubic>,<at 0.35 1.2 RCIM_Cubic>,<at 0.375 0 RCIM_Cubic>,<at 0.4 1.2 RCIM_Cubic>,<at 0.425 0 RCIM_Cubic>,<at 0.45 1.2 RCIM_Cubic>,<at 0.475 0 RCIM_Cubic>,<at 0.5 1.2 RCIM_Cubic>,<at 0.525 0 RCIM_Cubic>,<at 0.55 1.2 RCIM_Cubic>,<at 0.575 0 RCIM_Cubic>,<at 0.6 1.2 RCIM_Cubic>,<at 0.625 0 RCIM_Cubic>,<at 0.65 1.2 RCIM_Cubic>,<at 0.675 0 RCIM_Cubic>,<at 0.7 1.2 RCIM_Cubic>,<at 0.725 0 RCIM_Cubic>,<at 0.75 1.2 RCIM_Cubic>,<at 0.775 0 RCIM_Cubic>,<at 0.8 1.2 RCIM_Cubic>,<at 0.825 0 RCIM_Cubic>,<at 0.85 1.2 RCIM_Cubic>,<at 0.875 0 RCIM_Cubic>,<at 0.9 1.2 RCIM_Cubic>,<at 0.925 0 RCIM_Cubic>,<at 0.95 1.2 RCIM_Cubic>,<at 0.975 0 RCIM_Cubic>,<at 1 1.2 RCIM_Cubic>]}}}
       Scale Curve=0.8
      ))
     Vector Group 3:Vector3f=<0 0 0>
     Vector Group 4:Vector3f=<0 0 0>
     Vector Group 5:Vector3f=<0 0 0>
    ))
   &DynamicMaterialParameters(
    Index 0 Param 1=1
    Index 0 Param 2=&MakeCustomFloatFromBool(Bool=$User.bIsPainted, True Float=1, False Float=0)
    Module.Index0Param0Write=true
    Module.Index0Param1Write=true
    Module.Index0Param2Write=true
    Module.Index0Param3Write=true
    Module.Index3Param0Write=true
    Module.Index3Param1Write=true
    Module.Index3Param2Write=true
    Module.Index3Param3Write=true
    Write Parameter Index 3=true
   )
   &ScaleColor(
    Scale Alpha=&VC_FloatBySpawnGroup(
     SpawnGroup or ExecIndex:NiagaraInt32=0
     CustomInt:NiagaraInt32=0
     MaximumSpawnGroups:NiagaraInt32=3
     Float SG 0:NiagaraFloat=1
     Float SG 1:NiagaraFloat=&FloatFromCurve(FloatCurve=DataInterface() {propertyValues {Curve {keys [<at 0 1>,<at 1 0>]}}})
     Float SG 2:NiagaraFloat=1
     Float SG 3:NiagaraFloat=0
     Float SG 4:NiagaraFloat=0
     Float SG 5:NiagaraFloat=0
    )
    Scale RGB=&VC_VectorBySpawnGroup(
     SpawnGroup or ExecIndex:NiagaraInt32=0
     CustomInt:NiagaraInt32=0
     MaximumSpawnGroups:NiagaraInt32=3
     Vector Group 0:Vector3f=<0.7 0.7 0.7>
     Vector Group 1:Vector3f=&VectorFromFloat(Value=&MakeCustomFloatFromBool(Bool=$User.bIsPainted, True Float=$User.StarOutlineScalar, False Float=5))
     Vector Group 2:Vector3f=&VectorFromFloat(Value=&MakeCustomFloatFromBool(Bool=$User.bIsPainted, True Float=&Multiply_Float(A=$User.StarOutlineScalar, B=1), False Float=5))
     Vector Group 3:Vector3f=<0 0 0>
     Vector Group 4:Vector3f=<0 0 0>
     Vector Group 5:Vector3f=<0 0 0>
    )
    ScaleA=true
    ScaleRGB=true
   )
   /SetVariables(Particles.GlowScaleA:Vector3f=<3 3 3>) {isSetParameters=true}
   &SolveForcesAndVelocity(Clamp Velocity=false, Limit Acceleration=false)
  ]
  rendererClasses [&NiagaraMeshRendererProperties, &NiagaraSpriteRendererProperties]
  renderers [
   {rendererClass=&NiagaraMeshRendererProperties, rendererIndex=0}
   {rendererClass=&NiagaraMeshRendererProperties, rendererIndex=1}
   {rendererClass=&NiagaraMeshRendererProperties, rendererIndex=2}
   {rendererClass=&NiagaraMeshRendererProperties, rendererIndex=3}
   {rendererClass=&NiagaraMeshRendererProperties, rendererIndex=4}
   {rendererClass=&NiagaraMeshRendererProperties, rendererIndex=5}
   {rendererClass=&NiagaraSpriteRendererProperties, rendererIndex=6}
  ]
 }
 NE_VC_DriftTrail_Ribbon_Parent {
  emitterSpawnScript [
   /SetVariables(Emitter.No Painted:NiagaraBool=&Boolean_NotOperation(Bool=$User.bIsPainted)) {isSetParameters=true}
   /SetVariables(Emitter.Painted:NiagaraBool=&Boolean_NotOperation(Bool=$StackContext.No Painted)) {isSetParameters=true}
  ]
  emitterUpdateScript [
   &EmitterState()
   &VC_RateQualityHandler(Epic Spawn Rate=90, High Spawn Rate=60, Low Spawn Rate=30, Medium Spawn Rate=45, Use Custom Rate?=true)
   &VC_SpawnRibbon_v1(Number of Ribbons=2, Spawn Time=-1, SpawnRate=40)
  ]
  particleSpawnScript [
   &VC_PaintedVariantsOverride(
    Black=<rgba 0.05 0.05 0.05 1>
    BurntSienna=<rgba 0.3176 0.1864 0.1207 1>
    Cobalt=<rgba 0.16 0.3307 0.8 1>
    Crimson=<rgba 0.8021 0.1003 0.1649 1>
    DefaultColor=<rgba 0.35 0.35 0.35 1>
    ForestGreen=<rgba 0.2554 0.7 0.182 1>
    Grey=<rgba 0.44 0.455 0.5 1>
    Lime=<rgba 0.5708 0.76 0.1292 1>
    Orange=<rgba 0.76 0.2939 0.152 1>
    Pink=<rgba 0.6764 0.4251 0.8021 1>
    Purple=<rgba 0.3148 0.16 0.8 1>
    R1 Black:LinearColor=<rgba 0.2 0.2 0.2 0>
    R1 BurntSienna:LinearColor=<rgba 0.5 0.1058 0 1>
    R1 Cobalt:LinearColor=<rgba 0.12 0.12 2 1>
    R1 Crimson:LinearColor=<rgba 1 0 0.0271 1>
    R1 ForestGreen:LinearColor=<rgba 0 0.7 0.0354 1>
    R1 Grey:LinearColor=<rgba 0.5 0.5 0.5 1>
    R1 Lime:LinearColor=<rgba 1.1 2.2 0 1>
    R1 Orange:LinearColor=<rgba 2 0.448 0.06 1>
    R1 Pink:LinearColor=<rgba 2 0.6 1.6863 1>
    R1 Purple:LinearColor=<rgba 0.5335 0 1.3 1>
    R1 Saffron:LinearColor=<rgba 0.7 0.7 0.14 1>
    R1 SkyBlue:LinearColor=<rgba 0.03 0.6425 1.5 1>
    R1 TitaniumWhite:LinearColor=<rgba 1 1 1 1>
    R2 Black:LinearColor=<rgba 1 1 1 1>
    R2 BurntSienna:LinearColor=<rgba 7 2.6833 0 1>
    R2 Cobalt:LinearColor=<rgba 1.5 1.5647 5 1>
    R2 Crimson:LinearColor=<rgba 3 0 0.0813 1>
    R2 DefaultColor:LinearColor=<rgba 50 15 44.4803 1>
    R2 ForestGreen:LinearColor=<rgba 0 5 0.2526 1>
    R2 Grey:LinearColor=<rgba 5 5 5 1>
    R2 Lime:LinearColor=<rgba 2.5 5 0 1>
    R2 Orange:LinearColor=<rgba 5 1.744 0 1>
    R2 Pink:LinearColor=<rgba 5 1.5 3.7273 1>
    R2 Purple:LinearColor=<rgba 1.2312 0 3 1>
    R2 Saffron:LinearColor=<rgba 5 3.8333 0 1>
    R2 SkyBlue:LinearColor=<rgba 0 3.3269 5 1>
    R2 TitaniumWhite:LinearColor=<rgba 5 5 5 1>
    Saffron=<rgba 0.5765 0.5833 0.1167 1>
    SkyBlue=<rgba 0.1431 0.5846 0.9 1>
    TitaniumWhite=<rgba 0.5725 0.5725 0.5804 1>
    Use R1 Painted Overrides=true
    Use R2 Painted Overrides=true
   )
   &VC_InitializeRibbon_v1(
    Custom Offset=true
    Enable Skeletal Position=true
    Facing Axis:ENiagaraOrientationAxis=::YAxis
    Lifetime=0.6
    Local Vector Offset=&AddVectorToPosition(Position=<-32 0 -20>, Vector=$User.offset)
    Offset Amount=1
    Ribbon Width=60
   )
   &AddVelocity(Velocity=<-1200 0 0>) {module=AddVelocity001}
  ]
  particleUpdateScript [
   &FN_DisableTireFX(Mask Index Attribute:NiagaraInt32=$Particles.SpawnGroup, Mask Index FL=2, Mask Index FR=3, Mask Index RL=0, Mask Index RR=1)
   &ParticleState()
   &InheritSourceMovement(
    Applied Position Scale=&Multiply_VectorByFloat(
     Vector=<1 1 1>
     Float=&FloatFromCurve(FloatCurve=DataInterface() {propertyValues {Curve {keys [{interpMode=RCIM_Cubic, tangentMode=RCTM_Break, time=0, value=1}, {interpMode=RCIM_Cubic, tangentMode=RCTM_Break, time=1, value=0}]}}})
    )
    ApplyForce=false
    ApplyPosition=true
    ApplyPositionFalloff=false
   )
   &SolveForcesAndVelocity(Clamp Velocity=false, Limit Acceleration=false)
   &VC_UpdateRibbon_v1(Detach Locked Particle=true, Enable Skeletal Position=true, Locked Particle Detach Time=0)
   &ScaleRibbonWidth(Ribbon Width Scale=&FloatFromCurve(
     FloatCurve=DataInterface() {propertyValues {Curve {keys [<at 0 0.5>,<at 1 1.5>]}}}
     Scale Curve=&SelectFloatFromArray(Direct Array Index=$Particles.SpawnGroup, Float Selection Array=DataInterface() {propertyValues {FloatData [1,1]}})
    ))
   /SetVariables(
    Particles.DefaultColor:LinearColor=<rgba 0.7 0.7 0.7 0>
    Particles.OutlineColor:LinearColor=&MultiplyLinearColor(
     Linear Color=$Particles.R2 Color
     Scale Factor=&MakeVector4FromVectorAndFloat(
      "Vector (RGB)"=<1 1 1>
      "Float (Alpha)"=&FloatFromCurve(FloatCurve=DataInterface() {propertyValues {Curve {
          defaultValue=1
          keys [{interpMode=RCIM_Cubic, tangentMode=RCTM_User, time=0, value=1}, {interpMode=RCIM_Cubic, tangentMode=RCTM_User, time=0.9, value=0}]
         }}})
     )
    )
   ) {isSetParameters=true}
  ]
  rendererClasses [&NiagaraRibbonRendererProperties]
  renderers [
   {rendererClass=&NiagaraRibbonRendererProperties}
   {rendererClass=&NiagaraRibbonRendererProperties, rendererIndex=1}
   {rendererClass=&NiagaraRibbonRendererProperties, rendererIndex=2}
   {rendererClass=&NiagaraRibbonRendererProperties, rendererIndex=3}
  ]
 }
}
summary {
 systemName=NS_DriftTrail_PopStar_MP
 emitters [
  {
   emitterName=BCKP_mesh_headstarr001
   rendererClasses [&NiagaraMeshRendererProperties, &NiagaraSpriteRendererProperties]
  }
  {emitterName=BCKP_Sprites_ColouredStars, rendererClasses [&NiagaraSpriteRendererProperties]}
  {
   emitterName=BCKP_StaticStar
   rendererClasses [&NiagaraMeshRendererProperties, &NiagaraSpriteRendererProperties]
  }
  {emitterName=NE_VC_DriftTrail_Ribbon_Parent, rendererClasses [&NiagaraRibbonRendererProperties]}
 ]
 userVariables [
  User.bIsFrontendPreview {type=NiagaraBool}
  User.bIsPainted {type=NiagaraBool}
  User.ColorVariantID {type=NiagaraInt32}
  User.offset {type=Vector3f, description="offset from wheels(set through cosmetic template)", defaultValue {struct=Vector3f, value=<0 0 0>}}
  User.PaintedColor {type=LinearColor, description="", defaultValue {struct=LinearColor, value=<rgba 1 1 1 1>}}
  User.PaintedScalar {type=NiagaraFloat, defaultValue=1, description=""}
  User.RibbonHLScalar {type=NiagaraFloat, defaultValue=4, description=""}
  User.RibbonInsideColor {type=NiagaraFloat, defaultValue=1, description=""}
  User.RibbonPaintedScalar {type=NiagaraFloat, defaultValue=1, description=""}
  User.StarOutlineScalar {type=NiagaraFloat, defaultValue=2, description=""}
  User.StarsSpritesScalar {type=NiagaraFloat, defaultValue=1, description=""}
  User.WheelSize {type=NiagaraFloat, defaultValue=0.1, description=""}
 ]
}
